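# Animation Blueprint for the M_MED_Nosh male body cosmetic, adding secondary motion and corrective posing over a pose copied from the attached parent mesh.
# Weighted multi-bone constraints position the neck and chest goo parent bones, then spring-driven AnimDynamics nodes simulate the goo bones.
# Additive glute bone transforms take their alpha from thigh yaw, measured against the ref pose in the EventGraph.
# Calf rotation variables are not read by the AnimGraph.
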AnimBlueprint /BRCosmetics/Characters/Player/Male/Medium/Bodies/M_MED_Nosh/Meshes/M_MED_Nosh_AnimBP
target_skeleton=&/M_MED_Nosh_Skeleton
graphs [
 AnimGraph {nodes [
   Anim Dynamics - Bone: dyn_neck_goo_01_l(Alpha=0.65, ExternalForce="0,0,0") {
    angularDampingOverride=0.9999
    bGravityOverrideInSimSpace=true
    bOverrideAngularDamping=true
    gravityScale=0.35
    simulationSpace=RootRelative
    alphaBoolBlend {blendOption=Linear}
    boundBone {boneName=dyn_neck_goo_01_l}
    physicsBodyDefinitions [{
      sphereCollisionRadius=0
      boundBone {boneName=dyn_neck_goo_01_l}
      boxExtents=<10 10 10>
      constraintSetup {angularLimitsMax=<25 0 10>, angularLimitsMin=<-25 -25 -10>, angularTarget=<0 0 0>}
      localJointOffset=<0 0.1 -10>
     }]
   }
   n1 = "Constraint - dyn_parent_neck_gooA (PPPP)"!(
    ComponentPose=Local To Component(LocalPose=Copy Pose From Mesh() {bUseAttachedParent=true})
    Alpha=1
    ConstraintWeights_0=0.1417
    ConstraintWeights_1:float=0.1771
    ConstraintWeights_2:float=0.0244
    ConstraintWeights_3:float=0.6568
   ) -> (Pose) {
    alphaBoolBlend {blendOption=Linear}
    boneToModify {boneName=dyn_parent_neck_gooA}
    constraintSetup [
     {targetBone {boneName=spine_03}}
     {targetBone {boneName=spine_05}}
     {targetBone {boneName=clavicle_l}}
     {targetBone {boneName=dyn_neck_brace}}
    ]
    constraintWeights [1,1,1,1]
   }
   n2 = "Constraint - dyn_parent_neck_gooB (PPPP)"!(
    ComponentPose=@n1
    Alpha=1
    ConstraintWeights_0=0.0224
    ConstraintWeights_1:float=0.143
    ConstraintWeights_2:float=0.1712
    ConstraintWeights_3:float=0.6634
   ) -> (Pose) {
    alphaBoolBlend {blendOption=Linear}
    boneToModify {boneName=dyn_parent_neck_gooB}
    constraintSetup [
     {targetBone {boneName=clavicle_r}}
     {targetBone {boneName=spine_03}}
     {targetBone {boneName=spine_05}}
     {targetBone {boneName=dyn_neck_brace}}
    ]
    constraintWeights [1,1,1,1]
   }
   n3 = "Constraint - dyn_parent_chest_goo (PPP)"!(ComponentPose=@n2, Alpha=1, ConstraintWeights_0=0.1505, ConstraintWeights_1:float=0.1518, ConstraintWeights_2:float=0.6977) -> (Pose) {
    alphaBoolBlend {blendOption=Linear}
    boneToModify {boneName=dyn_parent_chest_goo}
    constraintSetup [{targetBone {boneName=spine_03}}, {targetBone {boneName=spine_05}}, {targetBone {boneName=dyn_neck_brace}}]
    constraintWeights [1,1,1]
   }
   n4 = "Constraint - dyn_parent_neck_goo_l (PPPPPPPPP)"!(
    ComponentPose=@n3
    Alpha=1
    ConstraintWeights_0=0.278
    ConstraintWeights_1:float=0.1419
    ConstraintWeights_2:float=0.0608
    ConstraintWeights_3:float=0.0357
    ConstraintWeights_4:float=0.1866
    ConstraintWeights_5:float=0.0162
    ConstraintWeights_6:float=0.0515
    ConstraintWeights_7:float=0.0028
    ConstraintWeights_8:float=0.2267
   ) -> (Pose) {
    alphaBoolBlend {blendOption=Linear}
    boneToModify {boneName=dyn_parent_neck_goo_l}
    constraintSetup [
     {targetBone {boneName=head}}
     {targetBone {boneName=clavicle_l}}
     {targetBone {boneName=spine_03}}
     {targetBone {boneName=spine_04}}
     {targetBone {boneName=spine_05}}
     {targetBone {boneName=upperarm_twist_01_l}}
     {targetBone {boneName=neck_02}}
     {targetBone {boneName=neck_01}}
     {targetBone {boneName=dyn_neck_brace}}
    ]
    constraintWeights [1,1,1,1,1,1,1,1,1]
   }
   n5 = Anim Dynamics - Bone: dyn_neck_gooB_01(ComponentPose=@n4, Alpha=1, ExternalForce="0,0,0") {
    angularDampingOverride=1
    angularSpringConstant=1500
    bAngularSpring=true
    bGravityOverrideInSimSpace=true
    bOverrideAngularDamping=true
    gravityScale=0.32
    simulationSpace=RootRelative
    alphaBoolBlend {blendOption=Linear}
    boundBone {boneName=dyn_neck_gooB_01}
    physicsBodyDefinitions [{
      sphereCollisionRadius=0
      boundBone {boneName=dyn_neck_gooB_01}
      boxExtents=<10 20 20>
      constraintSetup {angularLimitsMax=<0 10 0>, angularLimitsMin=<0 -10 -40>, angularTarget=<0 0 0>}
      localJointOffset=<10 0 0>
     }]
   }
   n6 = Anim Dynamics - Bone: dyn_neck_gooB_02(ComponentPose=@n5, Alpha=1, ExternalForce="0,0,0") {
    angularDampingOverride=1
    angularSpringConstant=1500
    bAngularSpring=true
    bGravityOverrideInSimSpace=true
    bOverrideAngularDamping=true
    gravityScale=0.35
    simulationSpace=RootRelative
    alphaBoolBlend {blendOption=Linear}
    boundBone {boneName=dyn_neck_gooB_02}
    physicsBodyDefinitions [{
      sphereCollisionRadius=0
      boundBone {boneName=dyn_neck_gooB_02}
      boxExtents=<10 20 20>
      constraintSetup {angularLimitsMax=<0 10 0>, angularLimitsMin=<0 -10 -40>, angularTarget=<0 0 0>}
      localJointOffset=<10 0 0>
     }]
   }
   n7 = Anim Dynamics - Bone: dyn_neck_gooA_01(ComponentPose=@n6, Alpha=1, ExternalForce="0,0,0") {
    angularDampingOverride=1
    angularSpringConstant=1500
    bAngularSpring=true
    bGravityOverrideInSimSpace=true
    bOverrideAngularDamping=true
    gravityScale=0.35
    simulationSpace=RootRelative
    alphaBoolBlend {blendOption=Linear}
    boundBone {boneName=dyn_neck_gooA_01}
    physicsBodyDefinitions [{
      sphereCollisionRadius=0
      boundBone {boneName=dyn_neck_gooA_01}
      boxExtents=<10 20 20>
      constraintSetup {angularLimitsMax=<0 10 0>, angularLimitsMin=<0 -10 -40>, angularTarget=<0 0 0>}
      localJointOffset=<10 0 0>
     }]
   }
   n8 = Anim Dynamics - Bone: dyn_neck_gooA_02(ComponentPose=@n7, Alpha=1, ExternalForce="0,0,0") {
    angularDampingOverride=1
    angularSpringConstant=1500
    bAngularSpring=true
    bGravityOverrideInSimSpace=true
    bOverrideAngularDamping=true
    gravityScale=0.33
    simulationSpace=RootRelative
    alphaBoolBlend {blendOption=Linear}
    boundBone {boneName=dyn_neck_gooA_02}
    physicsBodyDefinitions [{
      sphereCollisionRadius=0
      boundBone {boneName=dyn_neck_gooA_02}
      boxExtents=<10 20 25>
      constraintSetup {angularLimitsMax=<0 10 0>, angularLimitsMin=<0 -10 -40>, angularTarget=<0 0 0>}
      localJointOffset=<7 0 0>
     }]
   }
   n9 = Anim Dynamics - Bone: dyn_chest_goo_01(ComponentPose=@n8, Alpha=1, ExternalForce="0,0,0") {
    angularDampingOverride=1
    angularSpringConstant=1500
    bAngularSpring=true
    bGravityOverrideInSimSpace=true
    bOverrideAngularDamping=true
    gravityScale=0.35
    simulationSpace=RootRelative
    alphaBoolBlend {blendOption=Linear}
    boundBone {boneName=dyn_chest_goo_01}
    componentAppliedLinearAccClamp=<1100 1100 1100>
    componentLinearAccScale=<0.7 0.7 0.5>
    componentLinearVelScale=<0.7 0.7 0.5>
    physicsBodyDefinitions [{
      sphereCollisionRadius=0
      boundBone {boneName=dyn_chest_goo_01}
      boxExtents=<10 20 20>
      constraintSetup {angularLimitsMax=<0 15 0>, angularLimitsMin=<0 -15 -40>, angularTarget=<0 0 0>}
      localJointOffset=<7 0 0>
     }]
   }
   n10 = Anim Dynamics - Bone: dyn_chest_goo_02(ComponentPose=@n9, Alpha=1, ExternalForce="0,0,0") {
    angularDampingOverride=1
    angularSpringConstant=1500
    bAngularSpring=true
    bGravityOverrideInSimSpace=true
    bOverrideAngularDamping=true
    gravityScale=0.35
    simulationSpace=RootRelative
    alphaBoolBlend {blendOption=Linear}
    boundBone {boneName=dyn_chest_goo_02}
    componentAppliedLinearAccClamp=<1100 1100 1100>
    componentLinearAccScale=<0.7 0.7 0.5>
    componentLinearVelScale=<0.7 0.7 0.5>
    physicsBodyDefinitions [{
      sphereCollisionRadius=0
      boundBone {boneName=dyn_chest_goo_02}
      boxExtents=<10 20 26>
      constraintSetup {angularLimitsMax=<0 15 0>, angularLimitsMin=<0 -15 -40>, angularTarget=<0 0 0>}
      localJointOffset=<3 0 0>
     }]
   }
   n11 = Anim Dynamics - Bone: dyn_neck_goo_01_l(ComponentPose=@n10, Alpha=1, ExternalForce="0,0,0") {
    angularDampingOverride=1
    angularSpringConstant=1500
    bAngularSpring=true
    bGravityOverrideInSimSpace=true
    bOverrideAngularDamping=true
    gravityScale=0.35
    simulationSpace=RootRelative
    alphaBoolBlend {blendOption=Linear}
    boundBone {boneName=dyn_neck_goo_01_l}
    physicsBodyDefinitions [{
      sphereCollisionRadius=0
      boundBone {boneName=dyn_neck_goo_01_l}
      boxExtents=<10 15 25>
      constraintSetup {angularLimitsMax=<0 12 0>, angularLimitsMin=<0 -15 -40>, angularTarget=<0 0 0>}
      localJointOffset=<10 0 0>
     }]
   }
   n12 = Anim Dynamics - Bone: dyn_neck_goo_02_l(ComponentPose=@n11, Alpha=1, ExternalForce="0,0,0") {
    angularDampingOverride=1
    angularSpringConstant=1500
    bAngularSpring=true
    bGravityOverrideInSimSpace=true
    bOverrideAngularDamping=true
    gravityScale=0.31
    simulationSpace=RootRelative
    alphaBoolBlend {blendOption=Linear}
    boundBone {boneName=dyn_neck_goo_02_l}
    physicsBodyDefinitions [{
      sphereCollisionRadius=0
      boundBone {boneName=dyn_neck_goo_02_l}
      boxExtents=<10 20 25>
      constraintSetup {angularLimitsMax=<0 12 0>, angularLimitsMin=<0 -12 -40>, angularTarget=<0 0 0>}
      localJointOffset=<5 0 0>
     }]
   }
   n13 = "Transform (Modify) Bone - Bone: dyn_glute_l"(ComponentPose=@n12, Alpha=Get thigh_rot_l(), Translation="-3.750,-3.383,1.052", Rotation="0.000,0.000,-55", Scale="1.3,1.3,1.3") {
    rotationMode=BMM_Additive
    rotationSpace=BCS_BoneSpace
    scaleMode=BMM_Additive
    scaleSpace=BCS_BoneSpace
    translationMode=BMM_Additive
    translationSpace=BCS_BoneSpace
    alphaBoolBlend {blendOption=Linear}
    alphaScaleBiasClamp {bInterpResult=true, interpSpeedDecreasing=15, interpSpeedIncreasing=15}
    boneToModify {boneName=dyn_glute_l}
    rotation {roll=-55}
    scale=<1.3 1.3 1.3>
    translation=<-3.75 -3.383 1.052>
   }
   n14 = "Transform (Modify) Bone - Bone: dyn_glute_r"(ComponentPose=@n13, Alpha=Get thigh_rot_r(), Translation="3.750,0,0", Rotation="0.000,0.000,-40", Scale="1.25,1.25,1.25") {
    rotationMode=BMM_Additive
    rotationSpace=BCS_BoneSpace
    scaleMode=BMM_Additive
    scaleSpace=BCS_BoneSpace
    translationMode=BMM_Additive
    translationSpace=BCS_BoneSpace
    alphaBoolBlend {blendOption=Linear}
    boneToModify {boneName=dyn_glute_r}
    rotation {roll=-40}
    scale=<1.25 1.25 1.25>
    translation {x=3.75}
   }
   Output Pose(Result=Component To Local(ComponentPose=@n14)) {layerGroup=None}
   Reroute Node(InputPin:Wildcard) -> (OutputPin:Wildcard)
  ]}
 EventGraph {nodes [
   Event BlueprintUpdateAnimation()
   > Get!(execute:Exec) -> (then:Exec, else:Exec, Skeletal_Mesh:Skeletal Mesh Component Object Reference)
   > Sequence()
   n1 = Get Skeletal_Mesh() -> (Skeletal_Mesh:Skeletal Mesh Component Object Reference)
   n2 = BreakTransform(InTransform=GetDeltaTransformFromRefPose(self=@n1, BoneName=thigh_l, BaseName=pelvis))
   > Set thigh_rot_l(thigh_rot_l=MapRangeClamped(Value=BreakRotator(InRot=@n2.Rotation).Yaw, InRangeA=-20, InRangeB=-100))
   n3 = BreakTransform(InTransform=GetDeltaTransformFromRefPose(self=@n1, BoneName=thigh_r, BaseName=pelvis))
   > Set thigh_rot_r(thigh_rot_r=MapRangeClamped(Value=BreakRotator(InRot=@n3.Rotation).Yaw, InRangeA=-20, InRangeB=-100))
   Event BlueprintInitializeAnimation()
   > Set Skeletal_Mesh(Skeletal_Mesh:Skeletal Mesh Component Object Reference=GetOwnerSkeletalMeshComponent()) -> (Output_Get:Skeletal Mesh Component Object Reference)
   Reroute Node(InputPin:Exec) -> (OutputPin:Exec)
   n4 = Get Skeletal_Mesh()
   n5 = BreakTransform(InTransform=GetDeltaTransformFromRefPose(self=@n4, BoneName=calf_l, BaseName=thigh_l))
   > Set calf_rot_l(calf_rot_l:double=MapRangeClamped(Value=BreakRotator(InRot=@n5.Rotation).Yaw, InRangeA=0, InRangeB=115)) -> (Output_Get:double)
   n6 = BreakTransform(InTransform=GetDeltaTransformFromRefPose(self=@n4, BoneName=thigh_r, BaseName=calf_r))
   > Set calf_rot_r(calf_rot_r:double=MapRangeClamped(Value=BreakRotator(InRot=@n6.Rotation).Yaw, InRangeA=0, InRangeB=-115)) -> (Output_Get:double)
  ]}
]
variables [
 calf_rot_l {type=number}
 calf_rot_r {type=number}
 Skeletal_Mesh {type=SkeletalMeshComponent}
 thigh_rot_l {type=number}
 thigh_rot_r {type=number}
]
