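# Physics asset for the head part of the male medium Nosh player cosmetic, set up for physics on one three-bone hair chain (dyn_hair_06).
# Small capsules on the hair bones are joined in sequence, with angular limits tight between segments, slightly looser toward the tip.
# The chain root attaches to neck_02 with free swing and twist.
# Wide capsules cover both neck bones and the right clavicle only.
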PhysicsAsset /BRCosmetics/Characters/Player/Male/Medium/Bodies/M_MED_Nosh/Meshes/Parts/Male_Medium_Nosh_Head_Physics
bodies {
 clavicle_r [{shapeType=Capsule, length=9, radius=9.5, center=<-6.3212 1.5481 -2.1974>, rotation=<rot 0 -89.9996 110.8151>}]
 dyn_hair_06_a [{shapeType=Capsule, length=0.505, radius=0.505, rotation=<rot 0 -90 0>}]
 dyn_hair_06_b [{shapeType=Capsule, length=1.5, radius=1.755, center=<1.7052 -0.0001 0>, rotation=<rot -0.0001 -89.9996 -89.9999>}]
 dyn_hair_06_c [{shapeType=Capsule, length=1.6149, radius=2.1313, center=<2.2728 0.2729 0.0778>, rotation=<rot 7.6865 -90 -8.472>}]
 neck_01 [{shapeType=Capsule, length=1.25, radius=12.75, center=<5.2206 0 -0.8997>, rotation=<rot 0 -89.9999 90>}]
 neck_02 [{shapeType=Capsule, length=2.5, radius=11.8, center=<4.0169 -0.0002 0>}]
}
constraints [
 {bone1Name=dyn_hair_06_b, bone2Name=dyn_hair_06_a, swing1LimitDegrees=20, swing2LimitDegrees=20, twistLimitDegrees=5}
 {bone1Name=dyn_hair_06_c, bone2Name=dyn_hair_06_b, swing1LimitDegrees=30, swing2LimitDegrees=30, twistLimitDegrees=5}
 {
  bone1Name=dyn_hair_06_a
  bone2Name=neck_02
  swing1LimitDegrees=45
  swing1Motion=Free
  swing2LimitDegrees=45
  swing2Motion=Free
  twistLimitDegrees=45
  twistMotion=Free
 }
]
