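# Post-process style head AnimBlueprint for the Nosh male medium character, adding procedural motion to the dynamic hair.
# Each update maps head yaw relative to spine_05 to head_rot_back, scaled down when not targeting.
# That value blends an additive offset on dyn_hair_06_a, which feeds a RigidBody simulation using the head physics asset.
# AnimDynamics hair chains and arm-driven collar bone nodes exist but are not connected to the output pose.
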AnimBlueprint /BRCosmetics/Characters/Player/Male/Medium/Bodies/M_MED_Nosh/Meshes/Parts/Male_Medium_Nosh_Head_AnimBP
target_skeleton=/Game/Characters/Player/Common/Fortnite_Base_Head/Fortnite_Base_Head_Export_Skeleton
graphs [
 AnimGraph {nodes [
   n1 = Anim Dynamics - Bone: dyn_hair_06_c(Alpha=1, ExternalForce="0,0,0") {
    angularDampingOverride=1
    angularSpringConstant=1500
    bAngularSpring=true
    bGravityOverrideInSimSpace=true
    bOverrideAngularDamping=true
    gravityScale=0.35
    simulationSpace=RootRelative
    alphaBoolBlend {blendOption=Linear}
    alphaScaleBiasClamp {scale=0.7}
    boundBone {boneName=dyn_hair_06_c}
    componentLinearAccScale=<0.1 0.1 0.2>
    componentLinearVelScale=<0.1 0.1 0.2>
    physicsBodyDefinitions [{
      sphereCollisionRadius=0
      boundBone {boneName=dyn_hair_06_c}
      boxExtents=<10 10 10>
      constraintSetup {angularLimitsMax=<0 40 0>, angularLimitsMin=<0 -40 -60>, angularTarget=<1 0 0>}
      localJointOffset=<5 0 0>
     }]
   }
   Anim Dynamics - Bone: dyn_hair_06_b(ComponentPose=@n1, Alpha=1, ExternalForce="0,0,0") {
    angularDampingOverride=1
    angularSpringConstant=1500
    bAngularSpring=true
    bGravityOverrideInSimSpace=true
    bOverrideAngularDamping=true
    gravityScale=0.35
    simulationSpace=RootRelative
    alphaBoolBlend {blendOption=Linear}
    alphaScaleBiasClamp {scale=0.7}
    boundBone {boneName=dyn_hair_06_b}
    componentLinearAccScale=<0.1 0.1 0.2>
    componentLinearVelScale=<0.1 0.1 0.2>
    physicsBodyDefinitions [{
      sphereCollisionRadius=0
      boundBone {boneName=dyn_hair_06_b}
      boxExtents=<10 10 10>
      constraintSetup {angularLimitsMax=<0 40 0>, angularLimitsMin=<0 -40 -60>, angularTarget=<1 0 0>}
      localJointOffset=<5 0 0>
     }]
   }
   n2 = Anim Dynamics - Bone: dyn_hair_06_a(Alpha=1, ExternalForce="0,0,0") {
    angularDampingOverride=1
    angularSpringConstant=1500
    bAngularSpring=true
    bGravityOverrideInSimSpace=true
    bOverrideAngularDamping=true
    gravityScale=0.35
    simulationSpace=RootRelative
    alphaBoolBlend {blendOption=Linear}
    boundBone {boneName=dyn_hair_06_a}
    physicsBodyDefinitions [{
      sphereCollisionRadius=0
      boundBone {boneName=dyn_hair_06_a}
      boxExtents=<10 20 20>
      constraintSetup {angularLimitsMax=<0 15 0>, angularLimitsMin=<0 -15 -40>, angularTarget=<0 0 0>}
      localJointOffset=<10 0 0>
     }]
   }
   Anim Dynamics - Bone: dyn_hair_06_b(ComponentPose=@n2, Alpha=1, ExternalForce="0,0,0") {
    angularDampingOverride=1
    angularSpringConstant=1500
    bAngularSpring=true
    bGravityOverrideInSimSpace=true
    bOverrideAngularDamping=true
    bUseSphericalLimits=true
    gravityScale=0.35
    simulationSpace=RootRelative
    alphaBoolBlend {blendOption=Linear}
    boundBone {boneName=dyn_hair_06_b}
    physicsBodyDefinitions [{
      collisionType=CustomSphere
      sphereCollisionRadius=3
      boundBone {boneName=dyn_hair_06_b}
      boxExtents=<10 20 20>
      constraintSetup {angularLimitsMax=<0 25 0>, angularLimitsMin=<0 -25 -40>, angularTarget=<0 0 0>}
      localJointOffset=<10 0 0>
     }]
    sphericalLimits [{limitRadius=33, drivingBone {boneName=spine_05}, sphereLocalOffset=<-5 16 0>}]
   }
   n3 = "Transform (Modify) Bone - Bone: driver_collar_frt_01_r"(Alpha=1, Translation="0,0,0", Rotation="-40,-4,20", Scale="0.500,0.500,0.500") {
    rotationMode=BMM_Additive
    rotationSpace=BCS_BoneSpace
    scaleMode=BMM_Additive
    scaleSpace=BCS_BoneSpace
    alphaBoolBlend {blendOption=Linear}
    alphaScaleBiasClamp {bInterpResult=true, interpSpeedDecreasing=30, interpSpeedIncreasing=30}
    boneToModify {boneName=driver_collar_frt_01_r}
    rotation=<rot -40 -4 20>
    scale=<0.5 0.5 0.5>
   }
   "Transform (Modify) Bone - Bone: driver_collar_frt_02_r"(
    ComponentPose=@n3
    Alpha=Get arm_up_more_r() -> (arm_up_more_r:double)
    Translation="-10,-0.000,-0.000"
    Rotation="0,0,0"
    Scale="0.25,0.25,0.25"
   ) {
    rotationMode=BMM_Additive
    rotationSpace=BCS_BoneSpace
    scaleMode=BMM_Additive
    scaleSpace=BCS_BoneSpace
    translationMode=BMM_Additive
    translationSpace=BCS_BoneSpace
    alphaBoolBlend {blendOption=Linear}
    alphaScaleBiasClamp {bInterpResult=true, interpSpeedDecreasing=30, interpSpeedIncreasing=30}
    boneToModify {boneName=driver_collar_frt_02_r}
    scale=<0.25 0.25 0.25>
    translation {x=-10}
   }
   n4 = "Transform (Modify) Bone - Bone: dyn_hair_06_a"(
    ComponentPose=Local To Component(LocalPose=Input Pose())
    Alpha=Get head_rot_back() -> (head_rot_back:double)
    Translation="-5,-3,0.000"
    Rotation="0,0,0"
    Scale="1,1,1"
   ) {
    rotationMode=BMM_Additive
    rotationSpace=BCS_BoneSpace
    scaleMode=BMM_Additive
    scaleSpace=BCS_BoneSpace
    translationMode=BMM_Additive
    translationSpace=BCS_BoneSpace
    alphaBoolBlend {blendOption=Linear}
    alphaScaleBiasClamp {bInterpResult=true, interpSpeedDecreasing=15, interpSpeedIncreasing=15}
    boneToModify {boneName=dyn_hair_06_a}
    translation=<-5 -3>
   }
   n5 = RigidBody(ComponentPose=@n4) {
    bClampLinearTranslationLimitToRefPose=true
    bOverrideWorldGravity=true
    overridePhysicsAsset=&/Male_Medium_Nosh_Head_Physics
    simulationSpace=BaseBoneSpace
    alphaBoolBlend {blendOption=Linear}
    baseBoneRef {boneName=root}
    componentAppliedLinearAccClamp=<3000 3000 1000>
    componentLinearAccScale=<0.7 0.7 0.3>
    componentLinearVelScale=<0.7 0.7 0.3>
    overrideWorldGravity {z=-980}
   }
   Output Pose(Result=Component To Local(ComponentPose=@n5)) {layerGroup=None}
   Get arm_up_r() -> (arm_up_r:double)
  ]}
 EventGraph {nodes [
   Event BlueprintUpdateAnimation()
   > Get!(execute:Exec) -> (then:Exec, else:Exec, Mesh_To_Copy:Skeletal Mesh Component Object Reference)
   > n1 = Sequence!(execute) -> (then_0, then_1, then_2:Exec, then_3:Exec)
   n2 = Get Mesh_To_Copy() -> (Mesh_To_Copy:Skeletal Mesh Component Object Reference)
   n3 = BreakRotator(InRot=BreakTransform(InTransform=GetDeltaTransformFromRefPose(self=@n2, BoneName=upperarm_r, BaseName=spine_05)).Rotation)
   > Set arm_up_r(arm_up_r:double=MapRangeClamped(Value=@n3.Yaw, InRangeA=80, InRangeB=138)) -> (Output_Get:double)
   > Set arm_up_more_r(arm_up_more_r:double=MapRangeClamped(Value=@n3.Yaw, InRangeA=85, InRangeB=125)) -> (Output_Get:double)
   > Set arm_back_r(arm_back_r:double=MapRangeClamped(Value=@n3.Yaw, InRangeA=35, InRangeB=-20)) -> (Output_Get:double)
   n4 = Branch(execute=@n1.then_1, Condition=Get bIsTargeting())
   n5 = GetDeltaTransformFromRefPose(self=Get Mesh_To_Copy(), BoneName=head, BaseName=spine_05)
   n6 = MapRangeClamped(Value=BreakRotator(InRot=BreakTransform(InTransform=@n5).Rotation).Yaw, InRangeA=0, InRangeB=-55)
   > Set head_rot_back(head_rot_back:double=@n6) -> (Output_Get:double)
   Set head_rot_back(execute=@n4.else, head_rot_back=float * float(A=@n6, B=0.3))
   Event BlueprintInitializeAnimation()
   > Set Mesh_To_Copy(Mesh_To_Copy:Skeletal Mesh Component Object Reference=GetOwnerSkeletalMeshComponent()) -> (Output_Get:Skeletal Mesh Component Object Reference)
   TryGetPawnOwner()
  ]}
]
variables [
 arm_back_r {type=number}
 arm_up_more_r {type=number}
 arm_up_r {type=number}
 head_rot_back {type=number}
 Mesh_To_Copy {type=SkeletalMeshComponent}
]
